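; Burst gameplay cue for LemonExam melee combo 3 that places ground impact effects beneath the Fortnite player pawn's weapon.
; On burst, it line-traces straight down from the weapon_r socket, then nudges the hit point along the actor's right vector.
; It spawns a fading decal there, oriented by impact normal and actor forward, and moves a cue particle component to match.
; If the hit actor is a water body, it adds a splash.
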(fn OnBurst(Niagara)
  (params (MyTarget :type Actor%20Object%20Reference) (Parameters :type Gameplay%20Cue%20Parameters%20Structure%20%28by%20ref%29) (ParticleComponents :type Array%20of%20Niagara%20Particle%20System%20Component%20Object%20References) (AudioComponents :type Array%20of%20Audio%20Component%20Object%20References) (BurstCameraShakeInstance :type Camera%20Shake%20Base%20Object%20Reference) (BurstDecalInstance :type Decal%20Actor%20Object%20Reference))
  (bind _returnvalue (Transformation|GetActorForwardVector MyTarget))
  (bind _returnvalue_2 (Transformation|GetActorRightVector MyTarget))
  (bind _returnvalue_3 (* _returnvalue_2 25.0))
  (bind _outhit (Collision|LineTraceForObjects :Start (Transformation|GetSocketLocation (Variables|/Script/Engine.Character|GetMesh (Utilities|Casting|CastToFortPlayerPawnAthena MyTarget)) "weapon_r") :End (- (Transformation|GetSocketLocation (Variables|/Script/Engine.Character|GetMesh (Utilities|Casting|CastToFortPlayerPawnAthena MyTarget)) "weapon_r") "0.000000,0.000000,1000.000000") :ObjectTypes (Variables|Default|GetObject%20Types_0) :bTraceComplex false :DrawDebugType "None" :bIgnoreSelf true :TraceColor "(R=1.000000,G=0.000000,B=0.000000,A=1.000000)" :TraceHitColor "(R=0.000000,G=1.000000,B=0.000000,A=1.000000)" :DrawTime 9.0))
  (bind _returnvalue_4 (Rendering|Decal|SpawnDecalatLocation 0 "/LemonExam/Assets/FX/Attack123/MI_LemonExam_Melee_Decal.MI_LemonExam_Melee_Decal" (* (Variables|Default|GetDecal%20Size) 1.4) (+ (.Location (Collision|BreakHitResult _outhit)) _returnvalue_3) (Math|Rotator|MakeRotfromXY (.ImpactNormal (Collision|BreakHitResult _outhit)) _returnvalue) 2.0))
  (Rendering|Components|Decal|SetFadeOut _returnvalue_4 0.0 2.0 true)
  (Transformation|SetWorldLocationAndRotation (Utilities|Array|Get(acopy) ParticleComponents) (+ (.Location (Collision|BreakHitResult _outhit)) _returnvalue_3) (Math|Rotator|MakeRotfromXY (.ImpactNormal (Collision|BreakHitResult _outhit)) _returnvalue) false false)
  (if (.bSuccess (Utilities|Casting|CastToWaterBody (.HitActor (Collision|BreakHitResult _outhit))))
    (Effects|Components|ParticleSystem|SpawnFXAtLocation 0 "/Game/Effects/Niagara/Systems/Enviroment/Water/NS_WaterSplash_Medium.NS_WaterSplash_Medium" (+ (.Location (Collision|BreakHitResult _outhit)) _returnvalue_3)))
  (return))
