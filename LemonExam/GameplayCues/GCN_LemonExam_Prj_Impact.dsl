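; Burst gameplay cue for the impact of the LemonExam arc sword projectile, leaving a melee decal and fitting the impact effect to the projectile.
; After the parent burst runs, it casts the target to Prj_LemonExam_Arc and spawns a short-lived fading decal at the projectile's location and rotation.
; The sword Niagara component's Y scale drives the decal's width and the uniform scale of the first spawned particle component, which also takes the projectile's rotation.
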(fn OnBurst(EitherVFX-preferthisversion)
  (params (MyTarget :type Actor%20Object%20Reference) (Parameters :type Gameplay%20Cue%20Parameters%20Structure%20%28by%20ref%29) (ParticleComponents :type Array%20of%20FXSystem%20Component%20Object%20References) (AudioComponents :type Array%20of%20Audio%20Component%20Object%20References) (BurstCameraShakeInstance :type Camera%20Shake%20Base%20Object%20Reference) (BurstDecalInstance :type Decal%20Actor%20Object%20Reference))
  (bind _output (Utilities|Array|Get(acopy) ParticleComponents))
  (|CallParentFunction MyTarget Parameters ParticleComponents AudioComponents BurstCameraShakeInstance BurstDecalInstance)
  (bind _asprj_lemon_exam_arc (Utilities|Casting|CastToPrj_LemonExam_Arc MyTarget)
    (:then
      (bind _returnvalue (Rendering|Decal|SpawnDecalatLocation _asprj_lemon_exam_arc "/LemonExam/Assets/FX/Attack123/MI_LemonExam_Melee_Decal.MI_LemonExam_Melee_Decal" (Math|Vector|MakeVector 4e+02 (* (.y (Transformation|GetWorldScale (Variables|/LemonExam/Gameplay/Prj_LemonExam_Arc.Prj_LemonExam_Arc_C|GetNS_RumiSword_Prj _asprj_lemon_exam_arc))) 2e+02) 32.0) (Transformation|GetActorLocation _asprj_lemon_exam_arc) (Transformation|GetActorRotation _asprj_lemon_exam_arc) 2.0))
      (Rendering|Components|Decal|SetFadeOut _returnvalue 0.0 2.0 true)
      (Transformation|SetWorldRotation _output (Transformation|GetActorRotation _asprj_lemon_exam_arc) false false)
      (Transformation|SetWorldScale3D _output (Math|Conversions|ToVector(Float) (/ (.y (Transformation|GetWorldScale (Variables|/LemonExam/Gameplay/Prj_LemonExam_Arc.Prj_LemonExam_Arc_C|GetNS_RumiSword_Prj _asprj_lemon_exam_arc))) 1.35))))
    (:CastFailed
      _))
  (return))
